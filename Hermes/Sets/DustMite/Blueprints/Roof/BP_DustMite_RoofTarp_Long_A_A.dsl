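; Long roof tarp prop from the DustMite set that resets a player's fall height on contact, so landing on it avoids accumulated fall damage.
; When an actor begins overlapping FallBox1, it casts to PlayerPawn_Athena, stores the result in PawnRef and calls ResetFallingHeight on that pawn.
; The construction script only calls its parent, and the graph applies no bounce or launch.
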(event OnComponentBeginOverlap(FallBox1)
  (params (OverlappedComponent :type Primitive%20Component%20Object%20Reference) (OtherActor :type Actor%20Object%20Reference) (OtherComp :type Primitive%20Component%20Object%20Reference) (OtherBodyIndex :type Integer) (bFromSweep :type Boolean) (SweepResult :type Hit%20Result%20Structure%20%28by%20ref%29))
  (bind _asplayer_pawn_athena (Utilities|Casting|CastToPlayerPawn_Athena OtherActor))
  (Variables|Default|SetPawnRef _asplayer_pawn_athena)
  (Pawn|Character|ResetFallingHeight (Variables|Default|GetPawnRef)))

(fn ConstructionScript
  (|CallParentFunction))
